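; Parent audio controller for Fortnite vehicles that drives cosmetic-aware engine, ignition, shutdown, refuel and collision sounds.
; A cosmetics receiver supplies sound data and motor-sim config, falling back on controller or data-asset defaults.
; Engine sounds load through a budgeted registry, registering and unregistering on sleep or disable, and fading out on eject.
; An ordered motor-sim component stack models the motor, while gear shifts and blowoff fire triggers on the engine audio component.
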(event VehicleAudio|EventOnInit
  (bind _returnvalue (AudioController|GetEngineAudioComponent))
  (bind _vehiclecosmeticsaudioreceiver (Variables|Default|GetVehicleCosmeticsAudioReceiver))
  (VehicleCosmeticsAudio|BindToCosmeticsUpdates _vehiclecosmeticsaudioreceiver (AudioController|GetVehicleActor))
  (VehicleCosmeticsAudio|BindEventtoOnVehicleAudioCosmeticsFailed _vehiclecosmeticsaudioreceiver (EventDispatchers|CreateEvent :!event Custom|OnVehicleAudioCosmeticsDataFailed))
  (VehicleCosmeticsAudio|BindEventtoOnVehicleAudioCosmeticsStructUpdated _vehiclecosmeticsaudioreceiver (EventDispatchers|CreateEvent :!event Custom|OnCosmeticsStructUpdated))
  (CallFunction|ConfigureSoundData)
  (CallFunction|SetDefaultCosmeticData)
  (Vehicle|StartAudio)
  (Utilities|IsValid _returnvalue
    (:Is%20Valid
      (EventDispatchers|BindEventtoOnAudioFinished _returnvalue (EventDispatchers|CreateEvent :!event OnEngineAudioComponentFinished)))))

(event VehicleAudio|EventOnShutdown
  (bind _returnvalue (AudioController|GetEngineAudioComponent))
  (VehicleCosmeticsAudio|UnbindallEventsfromOnVehicleAudioCosmeticsFailed (Variables|Default|GetVehicleCosmeticsAudioReceiver))
  (Utilities|IsValid _returnvalue
    (:Is%20Valid
      (EventDispatchers|UnbindallEventsfromOnAudioFinished _returnvalue))
    (:Is%20Not%20Valid
      _))
  (Vehicle|EndAudio))

(event OnGearChangedEvent(MotorPhysicsSim)
  (params (NewGear :type Integer))
  (bind _returnvalue (AudioController|GetEngineAudioComponent))
  (bind (_vehiclecosmeticdata_enginesound _vehiclecosmeticdata_engineignitionsound _vehiclecosmeticdata_engineoffsound _vehiclecosmeticdata_gearshiftupsound _vehiclecosmeticdata_gearshiftdownsound _vehiclecosmeticdata_hornsound _vehiclecosmeticdata_bisip _vehiclecosmeticdata_bignoreshifting _vehiclecosmeticdata_bignoreexhaustpops) (Utilities|Struct|BreakVehicleAudioCosmeticDataStruct (Variables|GetVehicleCosmeticData_EngineSound)))
  (Private|CallOnGearChangedEvent NewGear _vehiclecosmeticdata_bignoreexhaustpops)
  (if (> NewGear (Variables|Private|GetPrevGear))
    (Audio|Parameter|ExecuteTriggerParameter _returnvalue "Fort.Vehicle.OnUpshift")
    (else
      (Audio|Parameter|ExecuteTriggerParameter _returnvalue "Fort.Vehicle.OnDownShift")))
  (Variables|Private|SetPrevGear NewGear))

(event EventEndPlay
  (params (EndPlayReason :type EEndPlayReason%20Enum :default "Destroyed"))
  ;[0.1369 0.6624 1] Engine Audio Loading
  (|CallParentFunction EndPlayReason)
  (CallFunction|UnregisterEngineSoundfromLoading))

(event Custom|Event%20On%20Disabled%20Changed
  (params (bIsDisabled :type Boolean))
  ;[0.1369 0.6624 1] Engine Audio Loading
  (if bIsDisabled
    (CallFunction|UnregisterEngineSoundfromLoading)))

(event EventBeginPlay
  ;[0.1369 0.6624 1] Engine Audio Loading
  (|CallParentFunction))

(event Custom|OnCosmeticsStructUpdated
  (params (CosmeticInfo :type Vehicle%20Audio%20Cosmetic%20Chooser%20Output%20Structure))
  ;[0.1369 0.6624 1] Engine Audio Loading
  (bind _value (Utilities|InstancedStruct|GetInstancedStructValue (Utilities|Struct|BreakVehicleAudioCosmeticChooserOutput CosmeticInfo))
    (:Valid
      (Variables|Default|SetCosmeticsLoaded true)
      (bind _output_get (Variables|Default|SetVehicleCosmeticData _value))
      (bind (_output_get_enginesound _output_get_engineignitionsound _output_get_engineoffsound _output_get_gearshiftupsound _output_get_gearshiftdownsound _output_get_hornsound _output_get_bisip _output_get_bignoreshifting _output_get_bignoreexhaustpops _output_get_motorsimconfigdata) (Utilities|Struct|BreakVehicleAudioCosmeticDataStruct _output_get))
      (CallFunction|RegisterEngineSoundforLoading)
      (MotorModel|ConfigureMotorSimComponents (AudioController|GetAudioMotorModelComponent) _output_get_motorsimconfigdata))))

(event BudgetedEngineAudioUser|EventOnEngineAssetLoaded
  (params (LoadedAsset :type Sound%20Base%20Object%20Reference))
  ;[0.1369 0.6624 1] Engine Audio Loading
  (Utilities|IsValid LoadedAsset
    (:Is%20Valid
      (bind _output_get (Variables|Default|SetLoadedEngineSound LoadedAsset))
      (Audio|Components|Audio|SetSound (AudioController|GetEngineAudioComponent) _output_get)
      (Variables|Private|SetbUnloadEngineSoundOnFadeOut false)
      (CallFunction|ToggleEngineSound) ; new metasound method
    )))

(event BudgetedEngineAudioUser|EventOnEngineAssetEjected
  (bind _returnvalue (AudioController|GetEngineAudioComponent))
  ;[0.1369 0.6624 1] Engine Audio Loading
  (Variables|Default|SetLoadedEngineSound 0)
  (Utilities|IsValid _returnvalue
    (:Is%20Valid
      (Audio|Components|Audio|FadeOut _returnvalue 0.25 0.0 "Linear" :!class /Script/Engine.AudioComponent))))

(event Custom|OnVehicleAudioCosmeticsDataFailed
  (params (FailureInfo_Reason :type EVehicleCosmeticsFailureReason%20Enum))
  ;[0.1369 0.6624 1] Engine Audio Loading
  (Variables|Default|SetNoCosmeticsEquipped (Utilities|Enum|Equal(Enum) "NoBodyEquipped" FailureInfo_Reason)))

(event Custom|OnBlowoff
  (params (BlowoffStrength :type Float%20%28single-precision%29))
  (Audio|Parameter|ExecuteTriggerParameter (AudioController|GetEngineAudioComponent) "Fort.Vehicle.OnBlowoff"))

(event OnRevLimiterHit(RevLimiterMotorSim)
  (Private|CallOnRevLimiterHit))

(fn AddMotorComponents
  (MotorModel|Reset (Variables|Default|GetMotorModel))
  (MotorModel|AddMotorSimComponentOrder (Variables|Default|GetThrottleStateMotorSim))
  (MotorModel|AddMotorSimComponentOrder (Variables|Default|GetResistanceMotorSim))
  (MotorModel|AddMotorSimComponentOrder (Variables|Default|GetReverseMotorSim))
  (MotorModel|AddMotorSimComponentOrder (Variables|Default|GetRevLimiterMotorSim))
  (MotorModel|AddMotorSimComponentOrder (Variables|Default|GetMotorPhysicsSim))
  (MotorModel|AddMotorSimComponentOrder (Variables|Default|GetVelocitySyncMotorSim)))

(fn AddMotorSimComponentOrder
  (params (InComponent :type Audio%20Motor%20Sim%20Interface))
  (bind _sortorder (Variables|Private|GetSortOrder))
  (MotorModel|AddMotorSimComponent (Variables|Default|GetMotorModel) InComponent _sortorder)
  (Math|Integer|IncrementInt _sortorder))

(fn AudioDebug
  (bind (_deltatime _speed _forwardspeed _sidespeed _upspeed _throttle _brake _surfacefrictionmodifier) (Utilities|Struct|BreakAudioMotorSimInputContext (State|GetCachedInputData (Variables|Default|GetMotorModel))))
  (bind (_bshifting _gear _virtualspeedkmh _rpm) (Utilities|Struct|BreakAudioMotorSimRuntimeContext (State|GetRuntimeInfo (Variables|Default|GetMotorModel))))
  (Rendering|Debug|FlushDebugStrings)
  (Rendering|Debug|DrawDebugString (+ (Transformation|GetActorLocation) "0.000000,0.000000,50.000000") (Utilities|String|ToString(Text) (Utilities|Text|FormatText "NSLOCTEXT(\"[B977402B98327AB92C5723D12BE9C4C8]\", \"E24A2F1249175943FE47CA970560B5FB\", \"Gear: {gear}\\r\\nRPM: {rpm}\\r\\nThrottle: {throttle}\\r\\nSpeed: {speed}\\r\\nSurf Friction: {friction}\\r\\nRedline: {redline}\")" _gear _rpm _throttle (InputContext|GetTotalSpeed) _surfacefrictionmodifier (Variables|Private|GetbShouldRedline))) 0 "(R=1.000000,G=0.791786,B=0.119792,A=1.000000)"))

(fn BindEvents
  (bind _returnvalue (AudioController|GetVehicleActor))
  (bind _returnvalue_2 (Vehicle|GetDriver _returnvalue))
  (bind _returnvalue_3 (AudioController|GetVehicleActor))
  (Utilities|FlowControl|Sequence
    (:then_0
      (Vehicle|BindEventtoOnSleepStateChangedDelegate _returnvalue_3 (EventDispatchers|CreateEvent :!event On%20Sleep%20State%20Change))
      (EventDispatchers|BindEventtoOnDisabledChangedDelegate _returnvalue_3 (EventDispatchers|CreateEvent :!event Event%20On%20Disabled%20Changed))
      (Vehicle|BindEventtoOnSpringCompressionDelegate _returnvalue_3 (EventDispatchers|CreateEvent :!event UpdateSpringCompressionAudio))
      (Vehicle|BindEventtoOnPawnEnteredVehicleSeat _returnvalue_3 (EventDispatchers|CreateEvent :!event PawnEnterVehicle))
      (Vehicle|BindEventtoOnPawnExitedVehicleSeat _returnvalue_3 (EventDispatchers|CreateEvent :!event PawnExitVehicle))
      (Bindings|FuelBindings)
      (AudioController|BindVehicleCollisionEvent 0 (Utilities|Struct|MakeVehicleCollisionAudioConfig (Variables|Variables|GetCollision%20Audio%20Config%20Pawn%20Collision%20Sound) (Variables|Default|GetCollisionAudio) (Variables|Variables|GetCollision%20Audio%20Config%20Max%20Vehicle%20Speed) 0.1 1.0 0.4 1.0 0.1 1.0))
      (ThrottleState|BindEventtoOnEngineBlowoff (Variables|Default|GetThrottleStateMotorSim) (EventDispatchers|CreateEvent :!event OnBlowoff)))
    (:then_1
      (Vehicle|OnSleepStateChange (not (Vehicle|IsAsleep _returnvalue)))
      (Utilities|IsValid _returnvalue_2
        (:Is%20Valid
          (Vehicle|PawnEnterVehicle _returnvalue _returnvalue_2))))))

(fn ConfigureSoundData
  (bind _vehiclesounddata (Variables|AudioController|GetVehicleSoundData)
    (:then
      (if (Variables|/Script/VehicleAudioRuntime.FortVehicleSoundData|GetbIgnoreModelData _vehiclesounddata)
        (Variables|/Script/VehicleAudioRuntime.FortVehicleSoundData|SetMaxBrakeThreshold (Variables|Data|GetMaxBrakeThreshold) _vehiclesounddata)
        (Variables|/Script/VehicleAudioRuntime.FortVehicleSoundData|SetThrottleMultiplier (Variables|Variables|GetThrottleMultiplier) _vehiclesounddata)
        (Variables|/Script/VehicleAudioRuntime.FortVehicleSoundData|SetSurfaceFrictionModifier (Variables|Variables|GetSurface%20Friction%20Modifier) _vehiclesounddata)
        (Variables|/Script/VehicleAudioRuntime.FortVehicleSoundData|SetMotorFrictionModifier (Variables|Variables|GetMotor%20Friction%20Modifier) _vehiclesounddata)
        (Variables|/Script/VehicleAudioRuntime.FortVehicleSoundData|SetFrictionJitter (Math|Vector2D|MakeVector2D (Variables|Variables|GetFrictionJitterMin) (Variables|Variables|GetFrictionJitterMax)) _vehiclesounddata)
        (Variables|/Script/VehicleAudioRuntime.FortVehicleSoundData|SetbCanShiftOverride (Variables|Private|GetCanShiftOverride) _vehiclesounddata)))
    (:else)))

(fn EndAudio
  (bind _motormodel (Variables|Default|GetMotorModel)
    (:then
      (MotorModel|StopOutput _motormodel))
    (:else)))

(fn FuelBindings
  (bind _returnvalue (AudioController|GetVehicleActor))
  ;[1 1 1] Fuel
  (EventDispatchers|BindEventtoOnOutOfFuelDelegate _returnvalue (EventDispatchers|CreateEvent :!event OnOutOfFuel))
  (EventDispatchers|BindEventtoOnRefuelStartDelegate _returnvalue (EventDispatchers|CreateEvent :!event OnRefuelStart))
  (EventDispatchers|BindEventtoOnRefuelStopDelegate _returnvalue (EventDispatchers|CreateEvent :!event OnRefuelStop)))

(fn GetBoost
  (return (Math|Float|SelectFloat (Math|Conversions|ToFloat(Boolean) (Class|FortVehicleInterface|IsBoosting (AudioController|GetVehicleActor))) (Variables|Variables|GetBoost) (>= (Math|Conversions|ToFloat(Boolean) (Class|FortVehicleInterface|IsBoosting (AudioController|GetVehicleActor))) (Variables|Variables|GetBoost)))))

(fn GetBraking
  (return 0.0))

(fn GetCanShift
  (bind (_vehiclecosmeticdata_enginesound _vehiclecosmeticdata_engineignitionsound _vehiclecosmeticdata_engineoffsound _vehiclecosmeticdata_gearshiftupsound _vehiclecosmeticdata_gearshiftdownsound _vehiclecosmeticdata_hornsound _vehiclecosmeticdata_bisip _vehiclecosmeticdata_bignoreshifting) (Utilities|Struct|BreakVehicleAudioCosmeticDataStruct (Variables|GetVehicleCosmeticData_EngineSound)))
  (return (not _vehiclecosmeticdata_bignoreshifting)))

(fn GetCosmeticDataStruct
  (return (Variables|Default|GetVehicleCosmeticData)))

(fn GetFowardSpeed
  (return 0.0))

(fn GetGrounded
  (return false))

(fn GetNormalizedRpm
  (bind (_bshifting _gear _virtualspeedkmh _rpm) (Utilities|Struct|BreakAudioMotorSimRuntimeContext (State|GetRuntimeInfo (Variables|Default|GetMotorModel))))
  (return _rpm))

(fn GetPlayerPerspectiveModel
  (if (Math|Boolean|ORBoolean (TouchPlatform|IsNintendoSwitch) (Fortnite|IsMobilePlatform) (Variables|AudioController|GetIsLocalPawnInVehicle))
    (return 0)
    (else
      (return (Utilities|Select :Index (Variables|Private|GetPlayerAffliation))))))

(fn GetSideSpeed
  (return 5e+01))

(fn GetThrottle
  (return 1.0))

(fn GetTotalSpeed
  (return (Vehicle|GetTotalSpeedKmh (AudioController|GetVehicleActor))))

(fn GetUpSpeed
  (bind (_returnvalue_x _returnvalue_y _returnvalue_z) (Math|Vector|BreakVector (Vehicle|GetVehicleVelocity (AudioController|GetVehicleActor))))
  (return (* (* (Math|Float|Absolute(Float) _returnvalue_z) 0.036) 0)))

(fn OnSleepStateChange
  (params (bIsAwake :type Boolean))
  (Actor|Tick|SetActorTickEnabled bIsAwake)
  (if bIsAwake
    (CallFunction|RegisterEngineSoundforLoading)
    (else
      (CallFunction|UnregisterEngineSoundfromLoading))))

(fn OnEngineAudioComponentFinished
  (bind _returnvalue (AudioController|GetEngineAudioComponent))
  (if (and (Variables|Private|GetbUnloadEngineSoundOnFadeOut) (Utilities|IsValid _returnvalue :!class /Script/Engine.KismetSystemLibrary))
    (Audio|Components|Audio|SetSound _returnvalue 0)))

(fn OnOutOfFuel
  (CallFunction|ToggleEngineSound)
  (CallFunction|PlayEngineOff))

(fn OnPassengersUpdated
  (Variables|AudioController|SetIsLocalPawnInVehicle false)
  (Variables|AudioController|SetLocalPawnInVehicle 0)
  (Variables|Private|SetIsTeamMemberInVehicle false)
  (for _array_element (Vehicle|GetAllPassengers (AudioController|GetVehicleActor))
    (bind _output_get (Variables|Private|SetPlayerAffliation (Team|GetTeamAffiliationforActortoLocalPlayer _array_element true true)))
    (switch Utilities|FlowControl|Switch|SwitchonEFortTeamAffiliation _output_get
      (:Friendly
        (Variables|Private|SetIsTeamMemberInVehicle true))
      (:Neutral
        _)
      (:Hostile
        _))
    (if (Pawn|IsLocallyViewed _array_element)
      (Variables|AudioController|SetIsLocalPawnInVehicle true)
      (Variables|AudioController|SetLocalPawnInVehicle _array_element)))
  (Audio|Parameter|SetBooleanParameter (AudioController|GetAudioParameterComponent) "bIsLocal" (Variables|AudioController|GetIsLocalPawnInVehicle) :!class /Script/AudioExtensions.AudioParameterControllerInterface))

(fn OnRefuelStart
  (Utilities|FlowControl|Sequence
    (:then_0
      (Vehicle|PlayRefuelSoundLoop))
    (:then_1
      (if (not (Audio|Components|Audio|IsPlaying (AudioController|GetEngineAudioComponent)))
        (CallFunction|ToggleEngineSound)))))

(fn OnRefuelStop
  (Vehicle|StopRefuelSoundLoop))

(fn PawnEnterVehicle
  (params (Vehicle :type Fort%20Vehicle%20Interface%20Interface%20%28by%20ref%29) (PlayerPawn :type Fort%20Pawn%20Object%20Reference) (SeatIndex :type Integer))
  (if (== SeatIndex 0)
    (CallFunction|ToggleEngineSound)
    (Variables|Variables|SetIgnitionWorldTime (Utilities|Time|GetGameTimeinSeconds))
    (if (> (Vehicle|Fuel|GetVehicleFuelPercent (AudioController|GetVehicleActor)) 0)
      (CallFunction|PlayIgnition)
      (else
        _))
    (Vehicle|OnPassengersUpdated)
    (else
      _)))

(fn PawnExitVehicle
  (params (Vehicle :type Fort%20Vehicle%20Interface%20Interface%20%28by%20ref%29) (PlayerPawn :type Fort%20Pawn%20Object%20Reference) (SeatIndex :type Integer))
  (bind _ignitionaudio (Variables|Default|GetIgnitionAudio))
  (if (== SeatIndex 0)
    (CallFunction|ToggleEngineSound)
    (Utilities|IsValid _ignitionaudio
      (:Is%20Valid
        (Audio|Components|Audio|FadeOut _ignitionaudio 0.2 0.0 "Linear" :!class /Script/Engine.AudioComponent))
      (:Is%20Not%20Valid
        _))
    (Variables|Default|SetIgnitionAudio 0)
    (if (> (Vehicle|Fuel|GetVehicleFuelPercent (AudioController|GetVehicleActor)) 0)
      (CallFunction|PlayEngineOff)
      (else
        _))
    (Vehicle|OnPassengersUpdated)
    (else
      _)))

(fn PlayEngineOff
  (bind (_vehiclecosmeticdata_enginesound _vehiclecosmeticdata_engineignitionsound _vehiclecosmeticdata_engineoffsound _vehiclecosmeticdata_gearshiftupsound _vehiclecosmeticdata_gearshiftdownsound _vehiclecosmeticdata_hornsound _vehiclecosmeticdata_bisip) (Utilities|Struct|BreakVehicleAudioCosmeticDataStruct (Variables|GetVehicleCosmeticData_EngineSound)))
  (bind _rootcomponent (Variables|Transformation|GetRootComponent))
  (Utilities|IsValid _vehiclecosmeticdata_engineoffsound
    (:Is%20Valid
      (Audio|SpawnSoundAttached _vehiclecosmeticdata_engineoffsound _rootcomponent))
    (:Is%20Not%20Valid
      (if (not _vehiclecosmeticdata_bisip)
        (Audio|SpawnSoundAttached (Variables|AudioController|GetDefaultEngineOff) _rootcomponent)))))

(fn PlayIgnition
  (bind _returnvalue (Utilities|Time|GetGameTimeinSeconds))
  (bind _ignitionworldtime (Variables|Variables|GetIgnitionWorldTime))
  (bind _returnvalue_2 (- _returnvalue _ignitionworldtime))
  (bind (_vehiclecosmeticdata_enginesound _vehiclecosmeticdata_engineignitionsound) (Utilities|Struct|BreakVehicleAudioCosmeticDataStruct (Variables|GetVehicleCosmeticData_EngineSound)))
  (bind _ignitiontimethreshold (Variables|Default|GetIgnitionTimeThreshold))
  (if (or (Variables|Default|GetCosmeticsLoaded) (Variables|Default|GetNoCosmeticsEquipped))
    _
    (elif (> _returnvalue_2 _ignitiontimethreshold)
      _
      (else
        (Utilities|Time|SetTimerbyEvent (EventDispatchers|CreateEvent :!event PlayIgnition) (- _ignitiontimethreshold _returnvalue_2)))))
  (bind _returnvalue_3 (Audio|SpawnSoundAttached _vehiclecosmeticdata_engineignitionsound (Variables|Transformation|GetRootComponent)))
  (Variables|Default|SetIgnitionAudio _returnvalue_3))

(fn PlayRefuelSoundLoop
  (Variables|AudioController|GetFuel%20Gained%20Loop
    (:then
      (if (Utilities|Enum|Equal(Enum) (Audio|Components|Audio|GetPlayState (Variables|AudioController|GetFuel%20Gained%20Loop)) "FadingOut")
        (Audio|Components|Audio|Play (Variables|AudioController|GetFuel%20Gained%20Loop) 0.0)))
    (:else
      (bind _returnvalue (Audio|SpawnSoundAttached (Variables|AudioController|GetGainFuelSound) (Variables|/Script/Engine.Actor|GetRootComponent (AudioController|GetVehicleActor))))
      (Variables|AudioController|SetFuel%20Gained%20Loop _returnvalue))))

(fn RegisterEngineSoundForLoading
  (bind _vehiclecosmeticdata_enginesound (Utilities|Struct|BreakVehicleAudioCosmeticDataStruct (Variables|GetVehicleCosmeticData_EngineSound)))
  (Class|FortVehicleEngineAudioRegistry|RegisterVehicleEntry (WorldSubsystems|GetFortVehicleEngineAudioRegistry) (Utilities|Struct|MakeVehicleEngineAudioRegistryEntryRequest self _vehiclecosmeticdata_enginesound)))

(fn RemoveMotorSimComponent
  (params (InComponent :type Audio%20Motor%20Sim%20Interface))
  (MotorModel|RemoveMotorSimComponent (Variables|Default|GetMotorModel) InComponent :!class /Script/AudioMotorSim.AudioMotorModelComponent)
  (Math|Integer|DecrementInt (Variables|Private|GetSortOrder)))

(fn SetAudioDataAsset
  (params (AudioDataAsset :type Vehicle%20Audio%20Controller%20Data%20Asset%20Object%20Reference))
  (bind _returnvalue (AudioController|GetEngineAudioComponent))
  ;[1 1 1] This is used for valet vehciles data types, should be temparory as we add the Data Asset directly to the variables so we can remove these variables on controller in the future.
  (bind _output_get (Variables|Default|SetAudio%20Data%20Asset AudioDataAsset))
  (Variables|AudioController|SetDefaultEngineSoundPath (Variables|/VehicleAudio/B_Vehicle_AudioController_PrimaryDataAsset.B_Vehicle_AudioController_PrimaryDataAsset_C|GetEngine%20Sound _output_get))
  (Variables|AudioController|SetVehicleSoundData (Variables|/VehicleAudio/B_Vehicle_AudioController_PrimaryDataAsset.B_Vehicle_AudioController_PrimaryDataAsset_C|GetComponent%20Data%20Asset _output_get))
  (Variables|AudioController|SetDefaultIgnition (Variables|/VehicleAudio/B_Vehicle_AudioController_PrimaryDataAsset.B_Vehicle_AudioController_PrimaryDataAsset_C|GetEngine%20On _output_get))
  (Variables|AudioController|SetDefaultEngineOff (Variables|/VehicleAudio/B_Vehicle_AudioController_PrimaryDataAsset.B_Vehicle_AudioController_PrimaryDataAsset_C|GetEngine%20Off _output_get))
  (Variables|AudioController|SetGainFuelSound (Variables|/VehicleAudio/B_Vehicle_AudioController_PrimaryDataAsset.B_Vehicle_AudioController_PrimaryDataAsset_C|GetGain%20Fuel%20Sound _output_get))
  (CallFunction|SetDefaultCosmeticData)
  (Variables|Variables|SetBoost (Variables|/VehicleAudio/B_Vehicle_AudioController_PrimaryDataAsset.B_Vehicle_AudioController_PrimaryDataAsset_C|GetBoost _output_get))
  (AudioControllerConfig|SetVehicleSoundData (Variables|/VehicleAudio/B_Vehicle_AudioController_PrimaryDataAsset.B_Vehicle_AudioController_PrimaryDataAsset_C|GetComponent%20Data%20Asset _output_get) true)
  (CallFunction|ConfigureSoundData)
  (Utilities|IsValid (Variables|/VehicleAudio/B_Vehicle_AudioController_PrimaryDataAsset.B_Vehicle_AudioController_PrimaryDataAsset_C|GetEngine%20Sound%20Attenuation _output_get)
    (:Is%20Valid
      (Utilities|IsValid _returnvalue
        (:Is%20Valid
          (Variables|/Script/Engine.AudioComponent|SetAttenuationSettings (Variables|/VehicleAudio/B_Vehicle_AudioController_PrimaryDataAsset.B_Vehicle_AudioController_PrimaryDataAsset_C|GetEngine%20Sound%20Attenuation _output_get) _returnvalue))))))

(fn SetDefaultCosmeticData
  (bind _vehiclecosmeticdata (Variables|Default|GetVehicleCosmeticData))
  (bind _vehiclesounddata (Variables|AudioController|GetVehicleSoundData))
  (Utilities|Struct|SetmembersinVehicleAudioCosmeticDataStruct _vehiclecosmeticdata (Variables|AudioController|GetDefaultEngineSoundPath) (Variables|AudioController|GetDefaultIgnition) (Variables|AudioController|GetDefaultEngineOff))
  (Utilities|IsValid _vehiclesounddata
    (:Is%20Valid
      (Utilities|Struct|SetmembersinVehicleAudioCosmeticDataStruct _vehiclecosmeticdata (Variables|/Script/VehicleAudioRuntime.FortVehicleSoundData|GetbCanShiftOverride _vehiclesounddata)))
    (:Is%20Not%20Valid
      _))
  (CallFunction|RegisterEngineSoundforLoading))

(fn SetRedlineState
  (params (Enabled :type Boolean))
  (Variables|Private|SetbShouldRedline Enabled))

(fn StartAudio
  (Bindings|BindEvents)
  (MotorModel|AddMotorComponents))

(fn StopRefuelSoundLoop
  (bind _fuel_gained_loop (Variables|AudioController|GetFuel%20Gained%20Loop)
    (:then
      (Audio|Components|Audio|FadeOut _fuel_gained_loop 1.0 0.0 "Linear" :!class /Script/Engine.AudioComponent))
    (:else)))

(fn ToggleEngineSound
  (AudioController|ToggleEngineSoundNative (Variables|Variables|GetEngineSoundFadeInDuration) (Variables|Variables|GetEngineSoundFadeOutDuration)))

(fn UnregisterEngineSoundFromLoading
  (bind _returnvalue (AudioController|GetEngineAudioComponent))
  (Variables|Private|SetbUnloadEngineSoundOnFadeOut true)
  (Class|FortVehicleEngineAudioRegistry|RemoveVehicleEntry (WorldSubsystems|GetFortVehicleEngineAudioRegistry) self)
  (Variables|Default|SetLoadedEngineSound 0)
  (if (Audio|Components|Audio|IsPlaying _returnvalue)
    (CallFunction|ToggleEngineSound)
    (else
      (Audio|Components|Audio|SetSound _returnvalue 0))))

(fn UpdateCollisionAudio
  (params (SelfActor :type Actor%20Object%20Reference) (OtherActor :type Actor%20Object%20Reference) (NormalImpulse :type Vector) (Hit_bBlockingHit :type Boolean) (Hit_bInitialOverlap :type Boolean) (Hit_Time :type Float%20%28single-precision%29) (Hit_Distance :type Float%20%28single-precision%29) (Hit_Location :type Vector :default "0, 0, 0") (Hit_ImpactPoint :type Vector :default "0, 0, 0") (Hit_Normal :type Vector :default "0, 0, 0") (Hit_ImpactNormal :type Vector :default "0, 0, 0") (Hit_PhysMat :type Physical%20Material%20Object%20Reference) (Hit_HitActor :type Actor%20Object%20Reference) (Hit_HitComponent :type Primitive%20Component%20Object%20Reference) (Hit_HitBoneName :type Name :default "None") (Hit_BoneName :type Name :default "None") (Hit_HitItem :type Integer) (Hit_ElementIndex :type Integer) (Hit_FaceIndex :type Integer) (Hit_TraceStart :type Vector :default "0, 0, 0") (Hit_TraceEnd :type Vector :default "0, 0, 0"))
  (AudioController|UpdateCollisionAudioNative OtherActor Hit_ImpactPoint))

(fn UpdateMotorModel
  (params (Delta%20Time :type Float%20%28single-precision%29)))

(fn UpdateSpringCompressionAudio
  (bind (_outresult _outmagnitude) (AudioController|UpdateSpringCompressionNative (Variables|Default|GetSpringCompressionAudio) (LocalVariables|GetVehicleLocationOffset)))
  (if _outresult
    (Variables|Private|SetCollisionMagnitude _outmagnitude)))

(fn ConstructionScript)
